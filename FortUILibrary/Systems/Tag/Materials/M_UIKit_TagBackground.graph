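# Translucent UI material that draws a tag background as a skewable rounded shape with an optional outline.
# MF_UI_CoreUI_RoundedMask builds fill and outline distance fields from scale, per-corner radii, skew angle and outline thickness.
# Both fields are nudged inward to keep skewed edges off the bounds, then anti-aliased to drive opacity and the fill-to-outline color blend.
# A floored hasOutline scalar acts as a switch for outline thickness and color.
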Material /FortUILibrary/Systems/Tag/Materials/M_UIKit_TagBackground
graph {
 properties {
  bUsedWithStaticMesh=false
  blendMode=BLEND_Translucent
  materialDomain=MD_UI
  pixelDepthOffsetMode=PDOM_Legacy
  usedShadingModels=Default Lit
 }
 nodes [
  n1 = ScalarParameter() {parameterName=hasOutline}
  n2 = MaterialFunctionCall(
   Size=MaterialFunctionCall(
    X=ScalarParameter() {defaultValue=0.85, group=Layout, parameterName=ScaleX, sortPriority=1}
    Y=ScalarParameter() {defaultValue=1, group=Layout, parameterName=ScaleY, sortPriority=1}
   ) {materialFunction=&MakeFloat2}.Result
   Corner Radius=MaterialFunctionCall(
    X=ScalarParameter() {group=Layout, parameterName=TopLeft, sortPriority=10}
    Y=ScalarParameter() {defaultValue=0.5, group=Layout, parameterName=BottomLeft, sortPriority=13}
    Z=ScalarParameter() {defaultValue=0.25, group=Layout, parameterName=TopRight, sortPriority=11}
    A=ScalarParameter() {group=Layout, parameterName=BottomRight, sortPriority=12}
   ) {materialFunction=&MakeFloat4}.Result
   Radius in Pixels=ScalarParameter() {group=Layout, parameterName=Radius in Pixels, sortPriority=9}
   Skew Angle=ScalarParameter() {defaultValue=10, group=Layout, parameterName=SkewAngle, sortPriority=4}
   Outline Thickness=Multiply(A=ScalarParameter() {defaultValue=4, parameterName=OutlineThickness}, B=Floor(None=@n1))
  ) {materialFunction=&MF_UI_CoreUI_RoundedMask}
  n3 = Add(A=@n2.Outline SDF) {constB=-0.01, desc=-.01 so when skewed it'll not be at the edge if scale is set to 1}
  n4 = ScalarParameter() {parameterName=hasOutline}
  LinearInterpolate(
   A=VectorParameter() {group=Color, parameterName=Color, defaultValue=<rgba 0.8836 0 1 1>}.RGB
   B=VectorParameter() {group=Color, parameterName=ColorOutline, defaultValue=<rgba 1 1 0 1>}.RGB
   Alpha=Multiply(
    A=MaterialFunctionCall(DistanceField=@n3) {materialFunction=&MF_UI_AntiAliasing}.Inner
    B=Floor(None=@n4)
   )
  ) {constA=0.09}
  Multiply(
   A=MaterialFunctionCall(DistanceField=Add(A=@n2.SDF) {constB=-0.01}) {materialFunction=&MF_UI_AntiAliasing}.Inner
   B=ScalarParameter() {defaultValue=1, parameterName=Opacity, sliderMax=1}
  )
 ]
}
